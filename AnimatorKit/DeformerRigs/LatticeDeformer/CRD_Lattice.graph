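# CRD_Lattice is a deformer Control Rig that gives animators a lattice cage for bending a mesh.
# Construction imports the skeleton and spawns a clamped-size grid of point controls under the LatticeDeformer control.
# Forward solve draws the cage, pivots rotated or scaled multi-selections around their center, and feeds lattice-space points, transform and weight to DG_LatticeDeformer.
# An optional bone mask takes its bone from the mask control's space-switched parent.
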ControlRigBlueprint /AnimatorKit/DeformerRigs/LatticeDeformer/CRD_Lattice
graphs [
 Add_1_ContainedGraph {nodes [
   n1 = RigVMFunctionEntryNode()
   Add = &RigVMFunction_MathDoubleAdd(A=@n1.A, B=@n1.B)
   RigVMFunctionReturnNode(Result=&RigVMFunction_MathDoubleAdd(A=@Add.Result, B=@n1.C).Result)
  ]}
 Add_2_1_ContainedGraph {nodes [
   n1 = RigVMFunctionEntryNode()
   RigVMFunctionReturnNode(Result=/Script/RigVM.RigVMFunction_MathIntAdd(A=&RigVMFunction_MathIntAdd(A=@n1.A, B=@n1.B) {template="Add::Execute(in A,in B,out Result)"}.Result, B=@n1.C) {is=struct RigVMUnitNode, template="Add::Execute(in A,in B,out Result)"}.Result)
  ]}
 Add_9_ContainedGraph {nodes [
   n1 = RigVMFunctionEntryNode()
   RigVMFunctionReturnNode(Result=&RigVMFunction_MathDoubleAdd(A=&RigVMFunction_MathDoubleAdd(A=@n1.A, B=@n1.B).Result, B=@n1.C).Result)
  ]}
 Concat_1_ContainedGraph {nodes [
   n1 = RigVMFunctionEntryNode()
   Concat = /Script/RigVM.RigVMFunction_NameConcat(A=@n1.A, B=@n1.B) {is=struct RigVMUnitNode, template="Concat::Execute(in A,in B,out Result)"}
   RigVMFunctionReturnNode(Result=&RigVMFunction_NameConcat(A=&RigVMFunction_NameConcat(A=@Concat.Result, B=@n1.C) {template="Concat::Execute(in A,in B,out Result)"}.Result, B=@n1.D) {template="Concat::Execute(in A,in B,out Result)"}.Result)
  ]}
 Multiply_1_1_ContainedGraph {nodes [
   n1 = RigVMFunctionEntryNode()
   RigVMFunctionReturnNode(Result=/Script/RigVM.RigVMFunction_MathVectorMul(B=@n1.C, A=&RigVMFunction_MathVectorMul(A=@n1.A, B=@n1.B) {template="Multiply::Execute(in A,in B,out Result)"}.Result) {is=struct RigVMUnitNode, template="Multiply::Execute(in A,in B,out Result)"}.Result)
  ]}
 New Function_ContainedGraph {nodes [
   n1 = RigVMFunctionEntryNode()
   Multiply = /Script/RigVM.RigVMFunction_MathIntMul(A=@n1.NumControls.X, B=@n1.NumControls.Y) {is=struct RigVMUnitNode, template="Multiply::Execute(in A,in B,out Result)"}
   Divide = /Script/RigVM.RigVMFunction_MathIntDiv(A=@n1.Index, B=@Multiply.Result) {is=struct RigVMUnitNode, template="Divide::Execute(in A,in B,out Result)"}
   Modulo = /Script/RigVM.RigVMFunction_MathIntMod(A=@n1.Index, B=@Multiply.Result) {is=struct RigVMUnitNode, template="Modulo::Execute(in A,in B,out Result)"}
   > RigVMFunctionReturnNode(
    Index3D=(X=0,Y=0,Z=0)
    Index3D.Z=@Divide.Result
    Index3D.Y=&RigVMFunction_MathIntDiv(B=@n1.NumControls.X, A=@Modulo.Result) {template="Divide::Execute(in A,in B,out Result)"}.Result
    Index3D.X=&RigVMFunction_MathIntMod(B=@n1.NumControls.X, A=@Modulo.Result) {template="Modulo::Execute(in A,in B,out Result)"}.Result
   )
  ]}
 New Function_ContainedGraph {nodes [
   n1 = RigVMFunctionEntryNode()
   Add_2 = RigVMAggregateNode(
    A=@n1.Index3D.X
    C=&RigVMFunction_MathIntMul(
     A=@n1.Index3D.Z
     B=&RigVMFunction_MathIntMul(A=@n1.NumControls.X, B=@n1.NumControls.Y) {template="Multiply::Execute(in A,in B,out Result)"}.Result
    ) {template="Multiply::Execute(in A,in B,out Result)"}.Result
    B=&RigVMFunction_MathIntMul(A=@n1.Index3D.Y, B=@n1.NumControls.X) {template="Multiply::Execute(in A,in B,out Result)"}.Result
   )
   > RigVMFunctionReturnNode(Index=@Add_2.Result)
  ]}
 RigVMFunctionLibrary {nodes [
   FromIndex3D = RigVMCollapseNode(Index3D=(X=0,Y=0,Z=0), NumControls=(X=0,Y=0,Z=0))
   ToIndex3D = RigVMCollapseNode(Index=0, NumControls=(X=0,Y=0,Z=0))
  ]}
 RigVMFunction_MathBoolAnd_1_ContainedGraph {nodes [
   n1 = RigVMFunctionEntryNode()
   RigVMFunction_MathBoolAnd = &RigVMFunction_MathBoolAnd(A=@n1.A, B=@n1.B)
   RigVMFunctionReturnNode(Result=&RigVMFunction_MathBoolAnd(B=@n1.C, A=@RigVMFunction_MathBoolAnd.Result).Result)
  ]}
 RigVMFunction_MathBoolAnd_6_ContainedGraph {nodes [
   n1 = RigVMFunctionEntryNode()
   RigVMFunctionReturnNode(Result=&RigVMFunction_MathBoolAnd(B=@n1.C, A=&RigVMFunction_MathBoolAnd(A=@n1.A, B=@n1.B).Result).Result)
  ]}
 RigVMFunction_Sequence_1_1_ContainedGraph {nodes [
   n1 = RigVMFunctionEntryNode()
   n2 = &RigVMFunction_Sequence(ExecuteContext=@n1.ExecuteContext)
   n3 = &RigVMFunction_Sequence(ExecuteContext=@n2.B)
   n4 = &RigVMFunction_Sequence(ExecuteContext=@n3.B)
   RigVMFunctionReturnNode(A=@n2.A, B=@n3.A, C=@n4.A, D=@n4.B)
  ]}
 RigVMModel {nodes [
   n1 = &RigUnit_PrepareForExecution()
   n2 = &RigVMFunction_Sequence(ExecuteContext=@n1.ExecutePin)
   n3 = &RigUnit_HierarchyImportFromSkeleton(NameSpace=None, bIncludeCurves=False, bIncludeMeshSockets=False, bIncludeVirtualBones=True, ExecutePin=@n2.A)
   Greater = /Script/RigVM.RigVMFunction_MathIntGreater(B=0, A=RigVMDispatchNode(Array=@n3.Items) {template="DISPATCH_RigVMDispatch_ArrayGetNum(in Array,out Num)"}.Num) {is=struct RigVMUnitNode, template="Greater::Execute(in A,in B,out Result)"}
   n4 = &RigVMFunction_ControlFlowBranch(Condition=@Greater.Result, ExecuteContext=@n3.ExecutePin)
   RigVMVariableNode(
    Value=RigVMDispatchNode(Index=0, Array=@n3.Items) {template="DISPATCH_RigVMDispatch_ArrayGetAtIndex(in Array,in Index,out Element)"}.Element.Name
    ExecuteContext=@n4.True
   ) {variable=RootBoneName}
   n5 = &RigVMFunction_Sequence(ExecuteContext=@n2.B)
   RigVMFunction_MathIntMake = /Script/RigVM.RigVMFunction_MathIntMake(Value=2) {is=struct RigVMUnitNode}
   n6 = &RigVMFunction_MathIntMake(Value=5)
   n7 = RigVMVariableNode() {variable=NumPoints}
   Clamp = &RigVMFunction_MathIntClamp(Minimum=@RigVMFunction_MathIntMake.Value, Maximum=@n6.Value, Value=@n7.Value.X)
   RigVMVariableNode(
    Value=(X=0,Y=0,Z=3)
    Value.X=@Clamp.Result
    Value.Y=&RigVMFunction_MathIntClamp(Minimum=@RigVMFunction_MathIntMake.Value, Maximum=@n6.Value, Value=@n7.Value.Y).Result
    Value.Z=&RigVMFunction_MathIntClamp(Minimum=@RigVMFunction_MathIntMake.Value, Maximum=@n6.Value, Value=@n7.Value.Z).Result
    ExecuteContext=@n5.A
   ) {variable=InternalNumPoints}
   > RigVMVariableNode(Value=&RigVMFunction_MathDoubleClamp(Maximum=10, Minimum=0.1, Value=RigVMVariableNode() {variable=PointSize}.Value).Result) {variable=PointSize}
   > RigVMVariableNode(Value=/Script/RigVM.RigVMFunction_MathVectorScale(Value=(X=0.2,Y=0.2,Z=0.2), Factor=RigVMVariableNode() {variable=PointSize}.Value) {is=struct RigVMUnitNode, template="Scale::Execute(in Value,in Factor,out Result)"}.Result) {variable=InitialPointShapeScale}
   > RigUnit_SetTransform = &RigUnit_SetTransform(
    Item=(Type=Control,Name="LatticeDeformer")
    Space=LocalSpace
    Weight=1
    bInitial=False
    Value=RigVMVariableNode() {variable=CurrentTransform}.Value
   )
   Multiply = &RigVMFunction_MathVectorMul(
    A=(X=60,Y=60,Z=60)
    B=/Script/RigVM.RigVMFunction_MathVectorClamp(Maximum=(X=1000,Y=1000,Z=1000), Minimum=(X=0.1,Y=0.1,Z=0.1), Value=RigVMVariableNode() {variable=Scale}.Value) {is=struct RigVMUnitNode, template="Clamp::Execute(in Value,in Minimum,in Maximum,out Result)"}.Result
   ) {template="Multiply::Execute(in A,in B,out Result)"}
   > RigVMVariableNode(Value=@Multiply.Result) {variable=Dimensions}
   n8 = &RigVMFunction_Sequence(ExecuteContext=@n5.B)
   /Script/ControlRig.RigUnit_HierarchySetShapeSettings(
    Item=(Type=Control,Name="LatticeDeformer")
    Settings=(bVisible=True,Name="Arrow4_Thin",Color=(R=1,G=1,B=0,A=1),Transform=(Rotation=(X=0,Y=0,Z=0,W=1),Translation=(X=0,Y=0,Z=0),Scale3D=(X=1,Y=1,Z=1)))
    ExecutePin=@n8.A
   ) {is=struct RigVMUnitNode}
   n9 = &RigVMFunction_Sequence(ExecuteContext=@n8.B)
   n10 = RigVMVariableNode() {variable=InternalNumPoints}
   RigVMFunction_ForLoopCount = /Script/RigVM.RigVMFunction_ForLoopCount(ExecutePin=@n9.A, Count=@n10.Value.Z) {is=struct RigVMUnitNode}
   > n11 = &RigVMFunction_ForLoopCount(Count=@n10.Value.Y)
   > n12 = &RigVMFunction_ForLoopCount(Count=@n10.Value.X)
   Int to Name = /Script/RigVM.RigVMFunction_MathIntToName(PaddedSize=0, Number=@n12.Index) {is=struct RigVMUnitNode, template="Int to Name::Execute(in Number,in PaddedSize,out Result)"}
   Int to Name_1 = &RigVMFunction_MathIntToName(PaddedSize=0, Number=@n11.Index) {template="Int to Name::Execute(in Number,in PaddedSize,out Result)"}
   Int to Name_1_1 = &RigVMFunction_MathIntToName(PaddedSize=0, Number=@RigVMFunction_ForLoopCount.Index) {template="Int to Name::Execute(in Number,in PaddedSize,out Result)"}
   Concat = RigVMAggregateNode(B=@"Int to Name".Result, C=@"Int to Name_1".Result, D=@"Int to Name_1_1".Result, A=RigVMVariableNode() {variable=ControlPointName}.Value)
   Item = /Script/ControlRig.RigUnit_Item(Item=(Type=Control,Name="Points")) {is=struct RigVMUnitNode}
   n13 = RigVMVariableNode() {variable=Dimensions}
   Scale = &RigVMFunction_MathVectorScale(Factor=0.5, Value=@n13.Value) {template="Scale::Execute(in Value,in Factor,out Result)"}
   Subtract = /Script/RigVM.RigVMFunction_MathVectorSub(
    A=(X=0,Y=0,Z=0)
    A.X=&RigVMFunction_MathDoubleMul(A=@n12.Ratio, B=@n13.Value.X).Result
    A.Y=&RigVMFunction_MathDoubleMul(A=@n11.Ratio, B=@n13.Value.Y).Result
    A.Z=&RigVMFunction_MathDoubleMul(A=@RigVMFunction_ForLoopCount.Ratio, B=@n13.Value.Z).Result
    B=@Scale.Result
   ) {is=struct RigVMUnitNode, template="Subtract::Execute(in A,in B,out Result)"}
   > SpawnControl = /Script/ControlRig.RigUnit_HierarchyAddControlTransform(
    InitialValue=(Rotation=(X=0,Y=0,Z=0,W=1),Translation=(X=0,Y=0,Z=0),Scale3D=(X=1,Y=1,Z=1))
    OffsetSpace=LocalSpace
    OffsetTransform=(Rotation=(X=0,Y=0,Z=0,W=1),Translation=(X=0,Y=0,Z=0),Scale3D=(X=1,Y=1,Z=1))
    Settings="(InitialSpace=LocalSpace,bUsePreferredRotationOrder=False,PreferredRotationOrder=YZX,Limits=(LimitTranslationX=(bMinimum=False,bMaximum=False),LimitTranslationY=(bMinimum=False,bMaximum=False),LimitTranslationZ=(bMinimum=False,bMaximum=False),LimitPitch=(b ...(truncated)"
    Name=@Concat.Result
    Parent=@Item.Item
    OffsetTransform.Translation=@Subtract.Result
    Settings.Shape.Color=RigVMVariableNode() {variable=PointColor}.Value
    Settings.Shape.Transform.Scale3D=RigVMVariableNode() {variable=InitialPointShapeScale}.Value
   ) {
    is=struct RigVMUnitNode
    template="SpawnControl::Execute(in InitialValue,in Settings,in OffsetTransform,in OffsetSpace,in Parent,in Name,out Item)"
   }
   n14 = /Script/ControlRig.RigUnit_CollectionChildrenArray(Parent=(Type=Control,Name="Points"), TypeToSearch=Control, bDefaultChildren=True, bIncludeParent=False, bRecursive=False) {is=struct RigVMUnitNode}
   > Set_Metadata = RigVMDispatchNode(
    Name=PointIndex
    NameSpace=Self
    Item=@SpawnControl.Item
    Value=/Script/RigVM.RigVMFunction_MathIntSub(B=1, A=RigVMDispatchNode(Array=@n14.Items) {template="DISPATCH_RigVMDispatch_ArrayGetNum(in Array,out Num)"}.Num) {is=struct RigVMUnitNode, template="Subtract::Execute(in A,in B,out Result)"}.Result
   ) {template="DISPATCH_RigDispatch_SetMetadata(in Item,in Name,in NameSpace,in Value,out Success)"}
   RigVMDispatchNode(
    Enabled=True
    Prefix="CRD_Lattice: "
    ScreenColor=(R=0.041,G=0.599,B=0,A=1)
    ScreenDuration=10
    Value="To change lattice size, right click the Control Rig Track -> Recreate Control Rig with New Settings"
    ExecuteContext=@n9.B
   ) {template="DISPATCH_RigVMDispatch_Print(in Prefix,in Value,in Enabled,in ScreenDuration,in ScreenColor)"}
   RigUnit_BeginExecution = &RigUnit_BeginExecution()
   n15 = &RigVMFunction_Sequence(ExecuteContext=@RigUnit_BeginExecution.ExecutePin)
   n16 = &RigVMFunction_Sequence(ExecuteContext=@n15.A)
   n17 = /Script/ControlRig.RigUnit_IsInteracting() {is=struct RigVMUnitNode}
   n18 = &RigVMFunction_ControlFlowBranch(Condition=@n17.bIsInteracting, ExecuteContext=@n16.A)
   Set_Num_3 = RigVMDispatchNode(Array=(), Num=0, ExecuteContext=@n18.True) {template="DISPATCH_RigVMDispatch_ArraySetNum(io Array,in Num)"}
   n19 = &RigVMFunction_Sequence(ExecuteContext=@Set_Num_3.ExecuteContext)
   For_Each_10 = RigVMDispatchNode(Array=&RigUnit_IsInteracting().Items, ExecuteContext=@n19.A) {template="DISPATCH_RigVMDispatch_ArrayIterator(in Array,out Element,out Index,out Count,out Ratio)"}
   n20 = &RigVMFunction_ControlFlowBranch(
    Condition=RigVMDispatchNode(
     B=(Type=Control,Name="Points")
     A=/Script/ControlRig.RigUnit_HierarchyGetParent(bDefaultParent=True, Child=@For_Each_10.Element) {is=struct RigVMUnitNode}.Parent
    ) {template="DISPATCH_RigVMDispatch_CoreEquals(in A,in B,out Result)"}.Result
    ExecuteContext=@For_Each_10.ExecuteContext
   )
   RigVMDispatchNode(Element=@For_Each_10.Element, ExecuteContext=@n20.True, Array=@Set_Num_3.Array) {template="DISPATCH_RigVMDispatch_ArrayAdd(io Array,in Element,out Index)"}
   n21 = RigVMDispatchNode(Array=@Set_Num_3.Array) {template="DISPATCH_RigVMDispatch_ArrayGetNum(in Array,out Num)"}
   n22 = &RigUnit_IsInteracting()
   RigVMFunction_MathBoolAnd_4 = RigVMAggregateNode(
    A=RigVMDispatchNode(B=@n21.Num, A=RigVMDispatchNode(Array=@n17.Items) {template="DISPATCH_RigVMDispatch_ArrayGetNum(in Array,out Num)"}.Num) {template="DISPATCH_RigVMDispatch_CoreEquals(in A,in B,out Result)"}.Result
    B=&RigVMFunction_MathIntGreater(B=0, A=@n21.Num) {template="Greater::Execute(in A,in B,out Result)"}.Result
    C=&RigVMFunction_MathBoolAnd(
     A=&RigVMFunction_MathBoolNot(Value=@n22.bIsTranslating).Result
     B=&RigVMFunction_MathBoolOr(A=@n22.bIsRotating, B=@n22.bIsScaling).Result
    ).Result
   )
   RigVMFunction_ControlFlowBranch = &RigVMFunction_ControlFlowBranch(Condition=@RigVMFunction_MathBoolAnd_4.Result, ExecuteContext=@n19.B)
   n23 = &RigVMFunction_Sequence(ExecuteContext=@RigVMFunction_ControlFlowBranch.True)
   n24 = &RigVMFunction_ControlFlowBranch(Condition=&RigVMFunction_MathBoolNot(Value=RigVMVariableNode() {variable=bWasInteracting}.Value).Result, ExecuteContext=@n23.A)
   Set_Num_5 = RigVMDispatchNode(Array=(), Num=0, ExecuteContext=@n24.True) {template="DISPATCH_RigVMDispatch_ArraySetNum(io Array,in Num)"}
   n25 = &RigVMFunction_Sequence(ExecuteContext=@Set_Num_5.ExecuteContext)
   For_Each_15 = RigVMDispatchNode(ExecuteContext=@n25.A, Array=@Set_Num_3.Array) {template="DISPATCH_RigVMDispatch_ArrayIterator(in Array,out Element,out Index,out Count,out Ratio)"}
   > RigVMDispatchNode(
    Element=&RigVMFunction_MathTransformMakeRelative(
     Global=(Rotation=(X=0,Y=0,Z=0,W=1),Translation=(X=0,Y=0,Z=0),Scale3D=(X=1,Y=1,Z=1))
     Global.Translation=&RigUnit_GetTransform(bInitial=False, Item=@For_Each_15.Element).Transform.Translation
     Parent=&RigUnit_GetTransform(Item=(Type=Control,Name="LatticeDeformer"), bInitial=False).Transform
    ).Local.Translation
    Array=@Set_Num_5.Array
   ) {template="DISPATCH_RigVMDispatch_ArrayAdd(io Array,in Element,out Index)"}
   RigVMVariableNode(
    Value=/Script/RigVM.RigVMFunction_MathVectorArrayAverage(Array=@Set_Num_5.Array) {is=struct RigVMUnitNode, template="ArrayAverage::Execute(in Array,out Average)"}.Average
    ExecuteContext=@n25.B
   ) {variable=CenterOfSelected}
   > For_Each_14 = RigVMDispatchNode(Array=@Set_Num_3.Array) {template="DISPATCH_RigVMDispatch_ArrayIterator(in Array,out Element,out Index,out Count,out Ratio)"}
   n26 = &RigVMFunction_MathTransformMakeRelative(
    Global=(Rotation=(X=0,Y=0,Z=0,W=1),Translation=(X=0,Y=0,Z=0),Scale3D=(X=1,Y=1,Z=1))
    Parent=(Rotation=(X=0,Y=0,Z=0,W=1),Translation=(X=0,Y=0,Z=0),Scale3D=(X=1,Y=1,Z=1))
    Global.Translation=&RigVMFunction_MathTransformMakeRelative(
     Global=(Rotation=(X=0,Y=0,Z=0,W=1),Translation=(X=0,Y=0,Z=0),Scale3D=(X=1,Y=1,Z=1))
     Global.Translation=&RigUnit_GetTransform(bInitial=False, Item=@For_Each_14.Element).Transform.Translation
     Parent=&RigUnit_GetTransform(Item=(Type=Control,Name="LatticeDeformer"), bInitial=False).Transform
    ).Local.Translation
    Parent.Translation=RigVMVariableNode() {variable=CenterOfSelected}.Value
   )
   > Set_Metadata_5 = RigVMDispatchNode(Name=Relative, NameSpace=Self, Item=@For_Each_14.Element, Value=@n26.Local) {template="DISPATCH_RigDispatch_SetMetadata(in Item,in Name,in NameSpace,in Value,out Success)"}
   n27 = &RigVMFunction_Sequence(ExecuteContext=@n23.B)
   n28 = &RigUnit_GetTransform(
    Space=LocalSpace
    bInitial=False
    Item=RigVMDispatchNode(Index=0, Array=@Set_Num_3.Array) {template="DISPATCH_RigVMDispatch_ArrayGetAtIndex(in Array,in Index,out Element)"}.Element
   )
   n29 = /Script/RigVM.RigVMFunction_MathTransformMakeAbsolute(
    Local=(Rotation=(X=0,Y=0,Z=0,W=1),Translation=(X=0,Y=0,Z=0),Scale3D=(X=1,Y=1,Z=1))
    Parent=(Rotation=(X=0,Y=0,Z=0,W=1),Translation=(X=0,Y=0,Z=0),Scale3D=(X=1,Y=1,Z=1))
    Local.Rotation=@n28.Transform.Rotation
    Local.Scale3D=@n28.Transform.Scale3D
    Parent.Translation=RigVMVariableNode() {variable=CenterOfSelected}.Value
   ) {is=struct RigVMUnitNode, template="Make Absolute::Execute(in Local,in Parent,out Global)"}
   n30 = &RigVMFunction_MathTransformMakeAbsolute(Local=@n29.Global, Parent=&RigUnit_GetTransform(Item=(Type=Control,Name="LatticeDeformer"), bInitial=False).Transform) {template="Make Absolute::Execute(in Local,in Parent,out Global)"}
   n31 = &RigUnit_GetTransform(Item=(Type=Control,Name="LatticeDeformer"), bInitial=False)
   Add_8 = RigVMAggregateNode(A=@n31.Transform.Scale3D.X, B=@n31.Transform.Scale3D.Y, C=@n31.Transform.Scale3D.Z)
   /Script/RigVM.RigVMFunction_DebugTransformMutableNoSpace(
    Color=(R=1,G=0,B=0,A=1)
    DebugDrawSettings=(DepthPriority=SDPG_Foreground,Lifetime=-1)
    Mode=Axes
    Thickness=0
    WorldOffset=(Rotation=(X=0,Y=0,Z=0,W=1),Translation=(X=0,Y=0,Z=0),Scale3D=(X=1,Y=1,Z=1))
    bEnabled=true
    Transform=@n30.Global
    Scale=&RigVMFunction_MathDoubleMul(B=10, A=&RigVMFunction_MathDoubleDiv(B=3, A=@Add_8.Result).Result).Result
    ExecutePin=@n27.A
   ) {is=struct RigVMUnitNode}
   > For_Each_16 = RigVMDispatchNode(Array=@Set_Num_3.Array) {template="DISPATCH_RigVMDispatch_ArrayIterator(in Array,out Element,out Index,out Count,out Ratio)"}
   n32 = &RigUnit_GetTransform(bInitial=False, Item=@For_Each_16.Element)
   Get_Metadata_9 = RigVMDispatchNode(
    Default=(Rotation=(X=0,Y=0,Z=0,W=1),Translation=(X=0,Y=0,Z=0),Scale3D=(X=1,Y=1,Z=1))
    Name=Relative
    NameSpace=Self
    Item=@For_Each_16.Element
   ) {template="DISPATCH_RigDispatch_GetMetadata(in Item,in Name,in NameSpace,in Default,out Value,out Found)"}
   > &RigUnit_SetTransform(
    Value=(Rotation=(X=0,Y=0,Z=0,W=1),Translation=(X=0,Y=0,Z=0),Scale3D=(X=1,Y=1,Z=1))
    Weight=1
    bInitial=False
    Item=@For_Each_16.Element
    Value.Rotation=@n32.Transform.Rotation
    Value.Scale3D=@n32.Transform.Scale3D
    Value.Translation=&RigVMFunction_MathTransformMakeAbsolute(Local=@Get_Metadata_9.Value, Parent=@n30.Global) {template="Make Absolute::Execute(in Local,in Parent,out Global)"}.Global.Translation
   )
   n33 = &RigVMFunction_Sequence(ExecuteContext=@n16.B)
   n34 = &RigVMFunction_ControlFlowBranch(
    Condition=&RigVMFunction_MathBoolNot(Value=&RigUnit_IsInteracting().bIsInteracting).Result
    ExecuteContext=@n33.A
   )
   For_Each_17 = RigVMDispatchNode(
    Array=&RigUnit_CollectionChildrenArray(Parent=(Type=Control,Name="Points"), TypeToSearch=Control, bDefaultChildren=True, bIncludeParent=False, bRecursive=False).Items
    ExecuteContext=@n34.True
   ) {template="DISPATCH_RigVMDispatch_ArrayIterator(in Array,out Element,out Index,out Count,out Ratio)"}
   > &RigUnit_SetTransform(
    Space=LocalSpace
    Value=(Rotation=(X=0,Y=0,Z=0,W=1),Translation=(X=0,Y=0,Z=0),Scale3D=(X=1,Y=1,Z=1))
    Weight=1
    bInitial=False
    Item=@For_Each_17.Element
    Value.Translation=&RigUnit_GetTransform(Space=LocalSpace, bInitial=False, Item=@For_Each_17.Element).Transform.Translation
   )
   n35 = &RigVMFunction_Sequence(ExecuteContext=@n33.B)
   RigVMVariableNode(Value=&RigUnit_IsInteracting().bIsInteracting, ExecuteContext=@n35.A) {variable=bWasInteracting}
   n36 = &RigVMFunction_Sequence(ExecuteContext=@n15.B)
   For_Each_7_1 = RigVMDispatchNode(
    Array=&RigUnit_CollectionChildrenArray(Parent=(Type=Control,Name="Points"), TypeToSearch=Control, bDefaultChildren=True, bIncludeParent=False, bRecursive=False).Items
    ExecuteContext=@n36.A
   ) {template="DISPATCH_RigVMDispatch_ArrayIterator(in Array,out Element,out Index,out Count,out Ratio)"}
   n37 = &RigUnit_GetTransform(Item=(Type=Control,Name="LatticeDeformer"), bInitial=False)
   Add = RigVMAggregateNode(A=@n37.Transform.Scale3D.X, B=@n37.Transform.Scale3D.Y, C=@n37.Transform.Scale3D.Z)
   RigVMFunction_MathVectorFromDouble = /Script/RigVM.RigVMFunction_MathVectorFromDouble(Value=&RigVMFunction_MathDoubleDiv(B=3, A=@Add.Result).Result) {is=struct RigVMUnitNode, template="Cast::Execute(in Value,out Result)"}
   Multiply_1 = RigVMAggregateNode(
    A=/Script/RigVM.RigVMFunction_MathVectorDiv(A=(X=1,Y=1,Z=1), B=&RigUnit_GetTransform(bInitial=False, Item=@For_Each_7_1.Element).Transform.Scale3D) {is=struct RigVMUnitNode, template="Divide::Execute(in A,in B,out Result)"}.Result
    B=@RigVMFunction_MathVectorFromDouble.Result
    C=RigVMVariableNode() {variable=InitialPointShapeScale}.Value
   )
   > /Script/ControlRig.RigUnit_SetShapeTransform(
    Transform=(Rotation=(X=0,Y=0,Z=0,W=1),Translation=(X=0,Y=0,Z=0),Scale3D=(X=1,Y=1,Z=1))
    Control=@For_Each_7_1.Element.Name
    Transform.Scale3D=@Multiply_1.Result
   ) {is=struct RigVMUnitNode}
   n38 = &RigVMFunction_Sequence(ExecuteContext=@n36.B)
   n39 = &RigUnit_CollectionChildrenArray(Parent=(Type=Control,Name="Points"), TypeToSearch=All, bDefaultChildren=True, bIncludeParent=False, bRecursive=False)
   For_Each = RigVMDispatchNode(Array=@n39.Items, ExecuteContext=@n38.A) {template="DISPATCH_RigVMDispatch_ArrayIterator(in Array,out Element,out Index,out Count,out Ratio)"}
   > n40 = &RigVMFunction_ForLoopCount(Count=3)
   ToIndex3D = RigVMFunctionReferenceNode(Index=@For_Each.Index, NumControls=RigVMVariableNode() {variable=InternalNumPoints}.Value)
   n41 = RigVMDispatchNode(Value=@ToIndex3D.Index3D) {template="DISPATCH_RigVMDispatch_Constant(io Value)"}
   n42 = RigVMDispatchNode(
    Values=(0,0,0)
    Values.0=&RigVMFunction_MathIntAdd(B=1, A=@n41.Value.X) {template="Add::Execute(in A,in B,out Result)"}.Result
    Values.1=@n41.Value.X
    Values.2=@n41.Value.X
    Index=@n40.Index
   ) {template="DISPATCH_RigVMDispatch_SelectInt32(in Index,in Values,out Result)"}
   n43 = RigVMVariableNode() {variable=InternalNumPoints}
   n44 = RigVMDispatchNode(Value=@ToIndex3D.Index3D) {template="DISPATCH_RigVMDispatch_Constant(io Value)"}
   n45 = RigVMDispatchNode(
    Values=(0,0,0)
    Values.1=&RigVMFunction_MathIntAdd(B=1, A=@n44.Value.Y) {template="Add::Execute(in A,in B,out Result)"}.Result
    Values.0=@n44.Value.Y
    Values.2=@n44.Value.Y
    Index=@n40.Index
   ) {template="DISPATCH_RigVMDispatch_SelectInt32(in Index,in Values,out Result)"}
   n46 = RigVMDispatchNode(Value=@ToIndex3D.Index3D) {template="DISPATCH_RigVMDispatch_Constant(io Value)"}
   n47 = RigVMDispatchNode(
    Values=(0,0,0)
    Values.2=&RigVMFunction_MathIntAdd(B=1, A=@n46.Value.Z) {template="Add::Execute(in A,in B,out Result)"}.Result
    Values.0=@n46.Value.Z
    Values.1=@n46.Value.Z
    Index=@n40.Index
   ) {template="DISPATCH_RigVMDispatch_SelectInt32(in Index,in Values,out Result)"}
   RigVMFunction_MathBoolAnd = RigVMAggregateNode(
    A=/Script/RigVM.RigVMFunction_MathIntLess(A=@n42.Result, B=@n43.Value.X) {is=struct RigVMUnitNode, template="Less::Execute(in A,in B,out Result)"}.Result
    B=&RigVMFunction_MathIntLess(A=@n45.Result, B=@n43.Value.Y) {template="Less::Execute(in A,in B,out Result)"}.Result
    C=&RigVMFunction_MathIntLess(A=@n47.Result, B=@n43.Value.Z) {template="Less::Execute(in A,in B,out Result)"}.Result
   )
   n48 = &RigVMFunction_ControlFlowBranch(ExecuteContext=@n40.ExecutePin, Condition=@RigVMFunction_MathBoolAnd.Result)
   FromIndex3D = RigVMFunctionReferenceNode(
    Index3D=(X=0,Y=0,Z=0)
    Index3D.X=@n42.Result
    Index3D.Y=@n45.Result
    Index3D.Z=@n47.Result
    NumControls=RigVMVariableNode() {variable=InternalNumPoints}.Value
   )
   /Script/RigVM.RigVMFunction_DebugLineNoSpace(
    DebugDrawSettings=(DepthPriority=SDPG_Foreground,Lifetime=-1)
    Thickness=0
    WorldOffset=(Rotation=(X=0,Y=0,Z=0,W=1),Translation=(X=0,Y=0,Z=0),Scale3D=(X=1,Y=1,Z=1))
    bEnabled=True
    B=&RigUnit_GetTransform(
     bInitial=False
     Item=RigVMDispatchNode(Array=@n39.Items, Index=@FromIndex3D.Index) {template="DISPATCH_RigVMDispatch_ArrayGetAtIndex(in Array,in Index,out Element)"}.Element
    ).Transform.Translation
    A=&RigUnit_GetTransform(bInitial=False, Item=@For_Each.Element).Transform.Translation
    ExecutePin=@n48.True
    Color=RigVMVariableNode() {variable=PointColor}.Value
   ) {is=struct RigVMUnitNode}
   n49 = &RigVMFunction_Sequence(ExecuteContext=@n38.B)
   n50 = &RigUnit_GetTransform(Item=(Type=Control,Name="LatticeDeformer"), bInitial=False)
   n51 = &RigUnit_CollectionChildrenArray(Parent=(Type=Control,Name="Points"), TypeToSearch=All, bDefaultChildren=True, bIncludeParent=False, bRecursive=False)
   RigVMFunction_MathBoxFromArray = /Script/RigVM.RigVMFunction_MathBoxFromArray(Array=/Script/RigVM.RigVMFunction_MathTransformArrayToSRT(Transforms=/Script/ControlRig.RigUnit_GetTransformItemArray(Space=GlobalSpace, bInitial=False, Items=@n51.Items) {is=struct RigVMUnitNode}.Transforms) {is=struct RigVMUnitNode}.Translations) {
    is=struct RigVMUnitNode
    template="ArrayBounds,CreateBox,CreateBoundingBox,NewBox,NewBoundingBoxMakeBox,MakeBoundingBox,BoundingBox,Bbox,Bounds::Execute(in Array,out Box,out Minimum,out Maximum,out Center,out Size)"
   }
   n52 = &RigVMFunction_MathDoubleDiv(B=2, A=@RigVMFunction_MathBoxFromArray.Size.Z)
   n53 = &RigVMFunction_MathDoubleDiv(B=4, A=@RigVMFunction_MathBoxFromArray.Size.Z)
   n54 = &RigVMFunction_MathTransformMakeRelative(
    Global=(Rotation=(X=0,Y=0,Z=0,W=1),Translation=(X=0,Y=0,Z=0),Scale3D=(X=1,Y=1,Z=1))
    Parent=@n50.Transform
    Global.Translation=&RigVMFunction_MathVectorSub(
     B=(X=0,Y=0,Z=0)
     B.Z=&RigVMFunction_MathDoubleAdd(A=@n52.Result, B=@n53.Result).Result
     A=@RigVMFunction_MathBoxFromArray.Center
    ) {template="Subtract::Execute(in A,in B,out Result)"}.Result
   )
   &RigUnit_SetShapeTransform(
    Control=LatticeDeformer
    Transform=(Rotation=(X=0,Y=0,Z=0,W=1),Translation=(X=0,Y=0,Z=0),Scale3D=(X=3,Y=3,Z=3))
    Transform.Rotation=@n54.Local.Rotation
    Transform.Translation=@n54.Local.Translation
    ExecutePin=@n49.A
   )
   n55 = /Script/ControlRig.RigUnit_HierarchyGetShapeSettings(Item=(Type=Control,Name="LatticeDeformer")) {is=struct RigVMUnitNode}
   > &RigVMFunction_DebugLineNoSpace(
    A=(X=0,Y=0,Z=0)
    DebugDrawSettings=(DepthPriority=SDPG_Foreground,Lifetime=-1)
    Thickness=0
    bEnabled=True
    Color=@n55.Settings.Color
    B=@n54.Local.Translation
    WorldOffset=@n50.Transform
   )
   > RigVMFunction_DebugTransformMutableNoSpace = &RigVMFunction_DebugTransformMutableNoSpace(
    DebugDrawSettings=(DepthPriority=SDPG_Foreground,Lifetime=-1)
    Mode=Box
    Scale=2
    Thickness=0
    Transform=(Rotation=(X=0,Y=0,Z=0,W=1),Translation=(X=0,Y=0,Z=0),Scale3D=(X=1,Y=1,Z=1))
    bEnabled=True
    Color=@n55.Settings.Color
    WorldOffset=@n50.Transform
   )
   n56 = &RigVMFunction_Sequence(ExecuteContext=@n49.B)
   /Script/ControlRig.RigUnit_SetControlVisibility(
    Item=(Type=Control,Name="Mask_BonePickerViaSpaceSwitch")
    bVisible=&RigVMFunction_MathBoolAnd(
     B=/Script/ControlRig.RigUnit_GetBoolAnimationChannel(Channel=IsVisible, Control=Mask_BonePickerViaSpaceSwitch, bInitial=False) {is=struct RigVMUnitNode, template="GetAnimationChannel::Execute(out Value,in Control,in Channel,in bInitial)"}.Value
     A=&RigUnit_GetBoolAnimationChannel(Channel=EnableBoneMask, Control=Mask_BonePickerViaSpaceSwitch, bInitial=False) {template="GetAnimationChannel::Execute(out Value,in Control,in Channel,in bInitial)"}.Value
    ).Result
    ExecutePin=@n56.A
   ) {is=struct RigVMUnitNode, template="SetControlVisibility::Execute(in Item,in Pattern,in bVisible)"}
   n57 = &RigVMFunction_Sequence(ExecuteContext=@n56.B)
   n58 = &RigUnit_HierarchyGetParent(Child=(Type=Control,Name="Mask_BonePickerViaSpaceSwitch"), bDefaultParent=false)
   RigVMVariableNode(
    Value=(Type=Bone,Name="None")
    Value.Name=RigVMDispatchNode(
     Condition=RigVMDispatchNode(B=Bone, A=@n58.Parent.Type) {template="DISPATCH_RigVMDispatch_CoreEquals(in A,in B,out Result)"}.Result
     True=@n58.Parent.Name
     False=RigVMVariableNode() {variable=RootBoneName}.Value
    ) {template="DISPATCH_RigVMDispatch_If(in Condition,in True,in False,out Result)"}.Result
    ExecuteContext=@n57.A
   ) {variable=BoneForMask}
   RigVMFunction_Sequence_1 = RigVMAggregateNode(ExecuteContext=@n57.B)
   n59 = &RigUnit_CollectionChildrenArray(Parent=(Type=Control,Name="Points"), TypeToSearch=All, bDefaultChildren=True, bIncludeParent=False, bRecursive=False)
   Num = RigVMDispatchNode(Array=@n59.Items) {template="DISPATCH_RigVMDispatch_ArrayGetNum(in Array,out Num)"}
   Set_Num = RigVMDispatchNode(Array=(), Num=@Num.Num, ExecuteContext=@RigVMFunction_Sequence_1.A) {template="DISPATCH_RigVMDispatch_ArraySetNum(io Array,in Num)"}
   > For_Each_1 = RigVMDispatchNode(Array=@n59.Items) {template="DISPATCH_RigVMDispatch_ArrayIterator(in Array,out Element,out Index,out Count,out Ratio)"}
   RigUnit_GetTransform = &RigUnit_GetTransform(Item=(Type=Control,Name="LatticeDeformer"), bInitial=False)
   > Set_At = RigVMDispatchNode(
    Index=@For_Each_1.Index
    Element=&RigVMFunction_MathTransformMakeRelative(Global=&RigUnit_GetTransform(bInitial=False, Item=@For_Each_1.Element).Transform, Parent=@RigUnit_GetTransform.Transform).Local.Translation
    Array=@Set_Num.Array
   ) {template="DISPATCH_RigVMDispatch_ArraySetAtIndex(io Array,in Index,in Element)"}
   RigVMFunction_Sequence = &RigVMFunction_Sequence(ExecuteContext=@RigVMFunction_Sequence_1.C)
   RigUnit_GetIntAnimationChannel = /Script/ControlRig.RigUnit_GetIntAnimationChannel(Channel=ExpandTowardsRoot, Control=Mask_BonePickerViaSpaceSwitch, bInitial=False) {is=struct RigVMUnitNode, template="GetAnimationChannel::Execute(out Value,in Control,in Channel,in bInitial)"}
   n60 = &RigUnit_GetBoolAnimationChannel(Channel=EnableBoneMask, Control=Mask_BonePickerViaSpaceSwitch, bInitial=False) {template="GetAnimationChannel::Execute(out Value,in Control,in Channel,in bInitial)"}
   n61 = &RigUnit_GetBoolAnimationChannel(Channel=InvertMask, Control=Mask_BonePickerViaSpaceSwitch, bInitial=False) {template="GetAnimationChannel::Execute(out Value,in Control,in Channel,in bInitial)"}
   RigUnit_GetFloatAnimationChannel = /Script/ControlRig.RigUnit_GetFloatAnimationChannel(Channel=Weight, Control=LatticeDeformer, bInitial=False) {is=struct RigVMUnitNode, template="GetAnimationChannel::Execute(out Value,in Control,in Channel,in bInitial)"}
   RigUnit_GetBoolAnimationChannel = &RigUnit_GetBoolAnimationChannel(Channel=EnableDebugDraw, Control=Mask_BonePickerViaSpaceSwitch, bInitial=False) {template="GetAnimationChannel::Execute(out Value,in Control,in Channel,in bInitial)"}
   /Script/OptimusCore.RigUnit_AddOptimusDeformer(
    DeformerGraphAsset=(DeformerGraph="&/DG_LatticeDeformer",Name="DeformerGraphAsset")
    Dimensions=(Value=(X=0,Y=0,Z=0),Name="Dimensions")
    Mask_BoneNames=(Value=("None"),Name="Mask_BoneNames")
    Mask_DebugDraw=(Value=False,Name="Mask_DebugDraw")
    Mask_ExpandTowardsLeaf=(Value=0,Name="Mask_ExpandTowardsLeaf")
    Mask_ExpandTowardsRoot=(Value=0,Name="Mask_ExpandTowardsRoot")
    Mask_Invert=(Value=0,Name="Mask_Invert")
    NumPoints=(Value=(X=0,Y=0,Z=0),Name="NumPoints")
    Points=(Value=,Name="Points")
    Settings=(ExecutionPhase=AfterDefaultDeformer,ExecutionGroup=1,DeformChildComponents=True,ExcludeChildComponentsWithTag="None",Name="Settings")
    Transform=(Value=(Rotation=(X=0,Y=0,Z=0,W=1),Translation=(X=0,Y=0,Z=0),Scale3D=(X=1,Y=1,Z=1)),Name="Transform")
    Weight=(Value=0,Name="Weight")
    Mask_ExpandTowardsRoot.Value=RigVMDispatchNode(False=0, True=@RigUnit_GetIntAnimationChannel.Value, Condition=@n60.Value) {template="DISPATCH_RigVMDispatch_If(in Condition,in True,in False,out Result)"}.Result
    Mask_ExpandTowardsLeaf.Value=RigVMDispatchNode(
     False=999
     True=&RigUnit_GetIntAnimationChannel(Channel=ExpandTowardsLeaf, Control=Mask_BonePickerViaSpaceSwitch, bInitial=False) {template="GetAnimationChannel::Execute(out Value,in Control,in Channel,in bInitial)"}.Value
     Condition=@n60.Value
    ) {template="DISPATCH_RigVMDispatch_If(in Condition,in True,in False,out Result)"}.Result
    Mask_Invert.Value=RigVMDispatchNode(False=0, True=1, Condition=&RigVMFunction_MathBoolAnd(B=@n61.Value, A=@n60.Value).Result) {template="DISPATCH_RigVMDispatch_If(in Condition,in True,in False,out Result)"}.Result
    Mask_BoneNames.Value.0=RigVMDispatchNode(False=RigVMVariableNode() {variable=RootBoneName}.Value, True=RigVMVariableNode() {variable=BoneForMask}.Value.Name, Condition=@n60.Value) {template="DISPATCH_RigVMDispatch_If(in Condition,in True,in False,out Result)"}.Result
    Weight.Value=@RigUnit_GetFloatAnimationChannel.Value
    Transform.Value=&RigUnit_GetTransform(Item=(Type=Control,Name="LatticeDeformer"), bInitial=False).Transform
    Mask_DebugDraw.Value=&RigVMFunction_MathBoolAnd(B=@RigUnit_GetBoolAnimationChannel.Value, A=@n60.Value).Result
    ExecutePin=@RigVMFunction_Sequence.A
    Points.Value=@Set_Num.Array
    Settings.ExecutionPhase=RigVMVariableNode() {variable=DeformerExecutionPhase}.Value
    Settings.DeformChildComponents=RigVMVariableNode() {variable=DeformChildComponents}.Value
    NumPoints.Value=RigVMVariableNode() {variable=InternalNumPoints}.Value
    Dimensions.Value=RigVMVariableNode() {variable=Dimensions}.Value
   ) {is=struct RigVMUnitNode}
   RigVMVariableNode(
    Value=&RigUnit_GetTransform(Item=(Type=Control,Name="LatticeDeformer"), bInitial=False).Transform
    ExecuteContext=@RigVMFunction_Sequence.B
   ) {variable=CurrentTransform}
  ]}
]
hierarchy [
 ComponentSpace {type=Null, loc=<0 0 0>, rot=<0 0 0 1>}
 LatticeDeformer {controlType=EULER_TRANSFORM, parent=ComponentSpace, type=Control, loc=<0 0 0>, rot=<0 0 0 1>}
 Points {controlType=EULER_TRANSFORM, parent=LatticeDeformer, type=Control, loc=<0 0 0>, rot=<0 0 0 1>}
 Weight {controlType=FLOAT, parent=LatticeDeformer, type=Control, loc=<1 0 0>, rot=<0 0 0 1>}
 Mask_BonePickerViaSpaceSwitch {controlType=EULER_TRANSFORM, parent=ComponentSpace, type=Control, loc=<0 0 0>, rot=<0 0 0 1>}
 EnableBoneMask {controlType=BOOL, parent=Mask_BonePickerViaSpaceSwitch, type=Control, loc=<0 0 0>, rot=<0 0 0 1>}
 EnableDebugDraw {controlType=BOOL, parent=Mask_BonePickerViaSpaceSwitch, type=Control, loc=<1 0 0>, rot=<0 0 0 1>}
 ExpandTowardsRoot {controlType=INTEGER, parent=Mask_BonePickerViaSpaceSwitch, type=Control, loc=<0 0 0>, rot=<0 0 0 1>}
 ExpandTowardsLeaf {controlType=INTEGER, parent=Mask_BonePickerViaSpaceSwitch, type=Control, loc=<20 0 0>, rot=<0 0 0 1>}
 InvertMask {controlType=BOOL, parent=Mask_BonePickerViaSpaceSwitch, type=Control, loc=<0 0 0>, rot=<0 0 0 1>}
 IsVisible_2 {controlType=BOOL, parent=Mask_BonePickerViaSpaceSwitch, type=Control, loc=<1 0 0>, rot=<0 0 0 1>}
]
variables [
 BoneForMask {type=FRigElementKey, default=(Type=Bone,Name="root")}
 CenterOfSelected {type=FVector, default=(X=0,Y=0,Z=0)}
 ControlPointName {type=FName, default=Point}
 CurrentTransform {type=FTransform, default=(Rotation=(X=0,Y=0,Z=0,W=1),Translation=(X=0,Y=0,Z=70),Scale3D=(X=1,Y=1,Z=1))}
 DeformChildComponents {type=bool, default=True}
 DeformerExecutionPhase {type=EOptimusDeformerExecutionPhase, default=AfterDefaultDeformer}
 Dimensions {type=FVector, default=(X=60,Y=60,Z=60)}
 InitialPointShapeScale {type=FVector, default=(X=1,Y=1,Z=1)}
 InternalNumPoints {type=FIntVector, default=(X=3,Y=3,Z=3)}
 MinPointPosZ {type=double, default=0}
 NumPoints {type=FIntVector, default=(X=3,Y=3,Z=3)}
 PointColor {type=FLinearColor, default=(R=0,G=0,B=1,A=1)}
 PointSize {type=double, default=1}
 RootBoneName {type=FName, default=None}
 Scale {type=FVector, default=(X=1,Y=1,Z=1)}
 bWasInteracting {type=bool, default=False}
]
